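# Additive unlit material that generates water caustics by scattering mesh quads as photon sprites.
# Surface normals come from a WaterNormal texture or Gerstner waves, with offset wave samples blended to hide seams.
# Vertex math refracts rays through those normals by water depth, wraps landing points, collapses each quad and stretches it along the chromatic dispersion axis.
# Sprite color combines a falloff mask and spectral spread texture, with brightness normalized by sprite area and tile count.
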Material /Water/Caustics/Materials/Generation/CausticsMeshScatter
graph {
 properties {
  bUsedWithInstancedStaticMeshes=true
  blendMode=BLEND_Additive
  pixelDepthOffsetMode=PDOM_Legacy
  shadingModel=MSM_Unlit
  twoSided=true
  usedShadingModels=Unlit
 }
 nodes [
  MaterialExpressionConstant3Vector_0 = Constant3Vector() {constant=<rgba 0 0 1 0>}
  n1 = TextureCoordinate() {coordinateIndex=1}
  n2 = Divide() {constA=127, constB=128}
  n3 = Constant3Vector()
  n4 = Divide(
   A=Subtract(A=TransformPosition(None=@n3) {transformType=TRANSFORMPOSSOURCE_World}, B=ActorPositionWS())
   B=ScalarParameter() {defaultValue=2048, parameterName=WorldSize}
  )
  n5 = ScalarParameter() {defaultValue=2, parameterName=Tiles}
  n6 = Floor(None=Multiply(A=Add(A=@n4) {constB=0.5}, B=@n5))
  n7 = Add(
   A=Divide(A=Divide(A=@n1, B=@n2) {constB=1, desc=Adjust for Mesh being 127 quads}, B=ScalarParameter() {defaultValue=2, parameterName=Tiles})
   B=ComponentMask(None=Divide(A=@n6, B=@n5)) {g=true, r=true}
  )
  n8 = TextureSampleParameter2D(UVs=@n7) {parameterName=WaterNormal, samplerType=SAMPLERTYPE_LinearColor, texture=/Water/Caustics/Textures/Generation/RT_DefaultNormal}
  n9 = ScalarParameter() {defaultValue=2048, parameterName=WorldSize}
  n10 = ScalarParameter() {parameterName=Caustics Overscan}
  n11 = MaterialFunctionCall(UVs=@n7, Texture Scale=Divide(A=@n9, B=Add(A=@n10, B=@n9))) {materialFunction=&ScaleUVsByCenter}
  n12 = ComponentMask(None=VectorParameter() {parameterName=OffsetB, defaultValue=<rgba 0 0.5 0 0>}.RGB) {g=true, r=true}
  n13 = Add(
   A=Frac(None=Add(A=@n11.UVs, B=@n12))
   B=ComponentMask(None=VectorParameter() {parameterName=OffsetE, defaultValue=<rgba 0 1 0 0>}.RGB) {g=true, r=true}
  )
  n14 = ScalarParameter() {defaultValue=2048, parameterName=WorldSize}
  n15 = ComponentMask(None=Divide(A=@n14)) {g=true, r=true}
  n16 = Fmod(
   A=LinearInterpolate(
    A=Time()
    B=CollectionParameter() {collection=/Water/Materials/MPC/MPC_Water, parameterName=ManualTime}
    Alpha=CollectionParameter() {collection=&MPC_Water, parameterName=FreezeTime}
   )
   B=Constant() {r=1000}
  )
  n17 = MaterialFunctionCall!(WaveParameterRT, Override Position=Subtract(A=Multiply(A=@n13, B=@n14), B=@n15) {constB=0}, Time=@n16) {materialFunction=/Water/Caustics/Materials/Functions/GetGerstnerWaves}
  n18 = Add(A=@n11.UVs, B=ComponentMask(None=VectorParameter() {parameterName=OffsetA, defaultValue=<rgba 0.5 0.5 0 0>}.RGB) {g=true, r=true}) {constB=0.5}
  n19 = VectorParameter() {parameterName=OffsetD, defaultValue=<rgba 9 0 0 0>}
  n20 = Multiply(A=Add(A=Frac(None=@n18), B=ComponentMask(None=@n19.RGB) {g=true, r=true}), B=@n14)
  n21 = MaterialFunctionCall(Float2=@n11.UVs) {materialFunction=&BreakOutFloat2Components}
  n22 = Constant() {r=0.5}
  n23 = ScalarParameter() {defaultValue=0.1, parameterName=CrossFixWidth}
  n24 = SphereMask(A=@n21.R, B=@n22, Radius=@n23) {hardnessPercent=0}
  n25 = LinearInterpolate(
   A=MaterialFunctionCall(Override Position=Subtract(A=@n20, B=@n15) {constB=0}, Time=@n16) {materialFunction=&GetGerstnerWaves}.Normal
   B=@n17.Normal
   Alpha=@n24
  )
  n26 = ComponentMask(None=VectorParameter() {parameterName=OffsetC, defaultValue=<rgba 0.5 0 0 0>}.RGB) {g=true, r=true}
  n27 = Add(
   A=Frac(None=Add(A=@n11.UVs, B=@n26))
   B=ComponentMask(None=VectorParameter() {parameterName=OffsetF, defaultValue=<rgba 0 5 0 0>}.RGB) {g=true, r=true}
  )
  n28 = MaterialFunctionCall(Override Position=Subtract(A=Multiply(A=@n27, B=@n14), B=@n15) {constB=0}, Time=@n16) {materialFunction=&GetGerstnerWaves}
  n29 = SphereMask(A=@n21.G, B=@n22, Radius=@n23) {hardnessPercent=0}
  MaterialExpressionScalarParameter_23 = ScalarParameter() {defaultValue=0.1, parameterName=sm}
  MaterialExpressionSaturate_4 = Saturate(None=@n11.UVs)
  n30 = Custom!(SplineMargin=@MaterialExpressionScalarParameter_23, UV=@MaterialExpressionSaturate_4) {
   code="//input float HeightScale;\r\n//input float SplineMargin;\r\n\r\nfloat SplineParameter1 = (0.5f - SplineMargin);\r\nfloat SplineParameter2 = (1.0f/SplineMargin);\r\n\r\n\r\n    // Re-scale [0..0.1] and [0.9..1] to [0..1]\r\n    float2 T        = max(abs(UV - 0.5f) - SplineParameter1, 0.0f) * SplineParameter2;\r\n    // Cubic interpolation\r\n    float2 F        = 2*T*T*T - 3*T*T + 1;\r\n    // Bilinear filter from 2D to 1D factor (making it zero at the edges)\r\n    float Scale        =  F.x * F.y;\r\n\r\nreturn Scale;"
   description=Soft Edge Mask
   inputs [
    {inputName=SplineMargin, input {expression=&self:MaterialExpressionScalarParameter_23}}
    {inputName=UV, input {expression=&self:MaterialExpressionSaturate_4}}
   ]
  }
  n31 = MaterialFunctionCall(Override Position=Subtract(A=Multiply(A=@n11.UVs, B=@n14), B=@n15) {constB=0}, Time=@n16) {materialFunction=&GetGerstnerWaves}
  n32 = LinearInterpolate(
   A=@n17.Normal
   B=LinearInterpolate(A=LinearInterpolate(A=@n25, B=@n28.Normal, Alpha=@n29), B=@n31.Normal, Alpha=@n30)
   Alpha=ScalarParameter() {defaultValue=1, parameterName=Seamless}
  )
  n33 = StaticBoolParameter() {parameterName=UseNormalMap}
  n34 = MaterialFunctionCall(Float3=Normalize(VectorInput=StaticSwitch(True=@n8.RGB, False=@n32, Value=@n33))) {materialFunction=&BreakOutFloat3Components}
  n35 = Max(A=@n34.B, B=ScalarParameter() {defaultValue=1, parameterName=Min Normal Z}) {constB=0.01}
  MaterialExpressionReroute_4 = Reroute(None=Normalize(VectorInput=MaterialFunctionCall(X=@n34.R, Y=@n34.R, Z=@n35) {materialFunction=&MakeFloat3}.Result))
  MaterialExpressionScalarParameter_7 = ScalarParameter() {defaultValue=1.3, parameterName=IOR}
  n36 = Custom!(RayDir=@MaterialExpressionConstant3Vector_0, Normal=@MaterialExpressionReroute_4, IOR=@MaterialExpressionScalarParameter_7) {
   code="return refract( RayDir , -Normal, 1/IOR);"
   description=Refract
   inputs [
    {inputName=RayDir, input {expression=&self:MaterialExpressionConstant3Vector_0}}
    {inputName=Normal, input {expression=&self:MaterialExpressionReroute_4}}
    {inputName=IOR, input {expression=&self:MaterialExpressionScalarParameter_7}}
   ]
  }
  n37 = VertexInterpolator(VS=@n36)
  Abs(None=@n37.PS)
  n38 = ScalarParameter() {parameterName=Backtrace, sliderMax=1}
  n39 = Multiply(
   A=ScalarParameter() {parameterName=WaterDepth}
   B=Multiply(A=ScalarParameter() {defaultValue=4, parameterName=Depth Adjust Material}, B=OneMinus(None=@n38))
  )
  n40 = Divide(A=ComponentMask(None=@n36) {g=true, r=true}, B=ComponentMask(None=@n36) {b=true})
  n41 = Multiply(A=@n40, B=@n39)
  n42 = Multiply(
   A=Subtract(A=@n7) {constB=0.5}
   B=Add(A=ScalarParameter() {parameterName=Caustics Overscan}, B=ScalarParameter() {defaultValue=2048, parameterName=WorldSize})
  )
  Add(A=@n42, B=@n41) {desc=worldpos + WPO}
  n43 = Constant3Vector() {constant=<rgba 0 0 64 0>}
  n44 = Subtract(A=@n42, B=@n41)
  n45 = Add(
   A=ScalarParameter() {defaultValue=2048, parameterName=WorldSize}
   B=Multiply(A=ScalarParameter() {parameterName=Caustics Overscan}, B=ScalarParameter() {defaultValue=1, parameterName=rewrewr})
  ) {constB=0}
  n46 = Floor(None=Add(A=Divide(A=@n44, B=@n45)) {constB=0.5})
  n47 = Multiply(A=@n45, B=ScalarParameter() {defaultValue=1, parameterName=Wrap})
  n48 = AppendVector(A=Subtract(A=@n44, B=Multiply(A=@n46, B=@n47)), B=Constant())
  n49 = Multiply(A=WorldPosition().XYZ, B=Constant3Vector() {constant=<rgba 1 1 0 0>})
  n50 = ScalarParameter() {defaultValue=2, parameterName=Chromatic Spread}
  n51 = Multiply(A=@n40, B=Add(A=Max(A=@n50) {constB=0.001}))
  MaterialExpressionMultiply_7 = Multiply(A=@n39, B=Distance(A=@n40, B=@n51))
  MaterialExpressionNormalize_0 = Normalize(VectorInput=Subtract(A=@n40, B=@n51))
  n52 = ScalarParameter() {defaultValue=2048, parameterName=WorldSize}
  n53 = Multiply(B=ScalarParameter() {defaultValue=2, parameterName=Tiles}) {constA=128}
  n54 = ScalarParameter() {defaultValue=1, parameterName=Photon Scale}
  n55 = Add(A=Multiply(A=Divide(A=@n52, B=@n53) {desc=World Texel Size}, B=Add(A=@n54)) {constB=2}) {constB=0}
  MaterialExpressionReroute_14 = Reroute(None=@n55)
  MaterialExpressionTextureCoordinate_3 = TextureCoordinate()
  MaterialExpressionScalarParameter_35 = ScalarParameter() {defaultValue=1, parameterName=test}
  MaterialExpressionScalarParameter_9 = ScalarParameter() {defaultValue=-1, parameterName=t2}
  n56 = Custom!(
   PhotonLength=@MaterialExpressionMultiply_7
   ChromaticVector=@MaterialExpressionNormalize_0
   SpriteSize=@MaterialExpressionReroute_14
   UV=@MaterialExpressionTextureCoordinate_3
   test=@MaterialExpressionScalarParameter_35
   test2=@MaterialExpressionScalarParameter_9
  ) {
   code="\r\n\r\nfloat2 xaxis = float2(-ChromaticVector.y, ChromaticVector.x);\r\n\r\n\r\nfloat2 wpo = 0;\r\n\r\nwpo += (PhotonLength+(SpriteSize*test)) * (ChromaticVector) * UV.x;\r\nwpo += xaxis * SpriteSize * (UV.y - 0.5);\r\n\r\nwpo += ChromaticVector * SpriteSize * 0.5*test2;\r\n\r\nreturn float3(wpo, 0);"
   description=Photon Spread
   inputs [
    {inputName=PhotonLength, input {expression=&self:MaterialExpressionMultiply_7}}
    {inputName=ChromaticVector, input {expression=&self:MaterialExpressionNormalize_0}}
    {inputName=SpriteSize, input {expression=&self:MaterialExpressionReroute_14}}
    {inputName=UV, input {expression=&self:MaterialExpressionTextureCoordinate_3}}
    {inputName=test, input {expression=&self:MaterialExpressionScalarParameter_35}}
    {inputName=test2, input {expression=&self:MaterialExpressionScalarParameter_9}}
   ]
  }
  n57 = ActorPositionWS()
  Add(
   A=Multiply(A=Add(A=@n43, B=@n31.WPO), B=ScalarParameter() {parameterName=Backtrace, sliderMax=1})
   B=Add(A=Subtract(A=@n48, B=@n49) {desc=Collapse photons to single points}, B=Add(A=@n56, B=@n57)) {desc=Expand Photons along chromatic aberration axis}
  )
  n58 = TextureCoordinate()
  AppendVector(A=Multiply(A=Subtract(A=@n58) {constB=0.5}, B=@MaterialExpressionReroute_14), B=Constant())
  Constant() {r=1}
  n59 = Subtract(A=ComponentMask(None=TextureCoordinate()) {r=true}) {constB=0.5}
  n60 = OneMinus(None=Multiply(A=Abs(None=@n59)) {constB=2})
  n61 = Divide(A=@n55) {constB=1}
  n62 = VertexInterpolator(VS=@MaterialExpressionMultiply_7)
  n63 = Saturate(None=Divide(A=@n61, B=Add(A=@n62.PS, B=@n61)))
  n64 = Saturate(None=Divide(A=Saturate(None=@n60), B=@n63))
  n65 = Multiply(A=Saturate(None=@n64)) {constB=0.5}
  n66 = ComponentMask(None=TextureCoordinate()) {g=true}
  n67 = TextureSample(UVs=MaterialFunctionCall(UVs=AppendVector(A=@n65, B=@n66), Texture Scale=ScalarParameter() {defaultValue=1.25, parameterName=MaskScale}) {materialFunction=&ScaleUVsByCenter}.UVs) {samplerType=SAMPLERTYPE_Grayscale, texture=/Engine/EngineVolumetrics/LightBeam/Materials/T_EV_LightBeam_Falloff_02}
  n68 = ScalarParameter() {defaultValue=1, parameterName=Brightness}
  n69 = Multiply(A=@n55, B=@n55)
  n70 = ScalarParameter() {defaultValue=2, parameterName=Tiles}
  n71 = Divide(A=Divide(A=Multiply(A=@n68) {constB=1000}, B=Max(A=@n69)) {constB=1}, B=Multiply(A=@n70, B=@n70))
  n72 = ComponentMask(None=TextureCoordinate()) {r=true}
  n73 = TextureSample(UVs=AppendVector(A=OneMinus(None=@n72), B=@n63)) {samplerType=SAMPLERTYPE_LinearColor, texture=/Water/Caustics/Textures/Generation/ChromaticSpreadTexture}
  n74 = Max(A=Max(A=@n73.R, B=@n73.R), B=@n73.B)
  n75 = Multiply(A=ComponentMask(None=VertexInterpolator(VS=@n71).PS) {r=true}, B=Divide(A=@n73.RGB, B=Max(A=@n74) {constB=0.001}))
  MaterialFunctionCall(X=ComponentMask(None=Multiply(A=@n75, B=@n67.R)) {r=true}, Y=@n30) {materialFunction=&MakeFloat3}
  MaterialFunctionCall!(
   Vector3=@n37.PS
   MaximumNumberOfDigits
   UVs=MaterialFunctionCall(UVs=TextureCoordinate(), Texture Scale=ScalarParameter() {defaultValue=2, parameterName=caaa}) {materialFunction=&ScaleUVsByCenter}.UVs
   DebugTextLocation RG_UpperRight BA_LowerLeft
   Component Spacing
  ) {materialFunction=/Engine/Functions/Engine_MaterialFunctions02/Utility/DebugFloat3Values}
  n76 = Subtract(A=WorldPosition().XYZ, B=ActorPositionWS())
  n77 = ScalarParameter() {defaultValue=2048, parameterName=WorldSize}
  Saturate(None=Add(A=Divide(A=@n76, B=@n77)) {constB=0.5})
  ScalarParameter() {defaultValue=2048, parameterName=WorldSize}
  ScalarParameter() {defaultValue=2048, parameterName=WorldSize}
  ScalarParameter() {defaultValue=16, parameterName=Sprite Width}
  ScalarParameter() {defaultValue=32, parameterName=TtesSize}
  SquareRoot(None=@n64)
  TextureCoordinate()
  TextureCoordinate()
  TextureSample(UVs=@n7) {samplerType=SAMPLERTYPE_Normal, texture=/Engine/Functions/Engine_MaterialFunctions02/ExampleContent/Textures/newhplaneNormalsMap}
  VertexInterpolator(VS=MaterialFunctionCall(X=@n24, Y=@n29, Z=@n30) {materialFunction=&MakeFloat3}.Result)
 ]
}
