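# Unlit preview material for generated caustics, showing an RT texture either as a single image or as an animated, tiled flipbook.
# The single image supports UV offset, optional saturation and an output divisor, while the flipbook derives its cell from time, playback FPS and SubUV frame counts.
# Scalar-driven lerps pick the view, optionally multiply it by MipRT's red channel, or replace it with MipRT's values drawn as debug digits.
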Material /Water/Caustics/Materials/Generation/CausticsPreview
graph {
 properties {pixelDepthOffsetMode=PDOM_Legacy, shadingModel=MSM_Unlit, usedShadingModels=Unlit}
 nodes [
  n1 = TextureSampleParameter2D(UVs=Add(A=TextureCoordinate(), B=ScalarParameter() {parameterName=UV Offset}) {constB=0.5}) {
   parameterName=RT
   samplerType=SAMPLERTYPE_LinearColor
   texture=/Engine/Functions/Engine_MaterialFunctions02/PivotPainter2/Black_1x1_EXR_Texture
  }
  n2 = Saturate(None=@n1.RGB)
  n3 = ScalarParameter() {parameterName=Clamp}
  n4 = Frac(None=Multiply(A=TextureCoordinate(), B=ScalarParameter() {defaultValue=2, parameterName=Tiling}))
  n5 = ComponentMask(None=VectorParameter() {parameterName=SubUVFrames, defaultValue=<rgba 8 8 0 0>}.RGB) {g=true, r=true}
  n6 = ScalarParameter() {defaultValue=15, parameterName=Caustics Playback FPS}
  n7 = VectorParameter() {parameterName=SubUVFrames, defaultValue=<rgba 8 8 0 0>}
  n8 = MaterialFunctionCall!(
   1D Index=Fmod(
    A=MaterialFunctionCall(Speed=@n6) {materialFunction=&TimeWithSpeedVariable}.Time
    B=Multiply(A=@n7.R, B=@n7.R)
   )
   Cells XY=@n7.RGB
  ) {materialFunction=/Engine/Functions/Engine_MaterialFunctions01/Coordinates/1Dto2DIndex}
  n9 = ComponentMask(None=@n7.RGB) {g=true, r=true}
  n10 = TextureSampleParameter2D(UVs=Add(A=Divide(A=@n4, B=@n5), B=Divide(A=@n8.2D Index, B=@n9))) {
   parameterName=RT
   samplerType=SAMPLERTYPE_LinearColor
   texture=&Black_1x1_EXR_Texture
  }
  n11 = LinearInterpolate(
   A=Divide(A=LinearInterpolate(A=@n1.RGB, B=@n2, Alpha=@n3), B=ScalarParameter() {defaultValue=1, parameterName="Output  Divisor"})
   B=@n10.RGB
   Alpha=ScalarParameter() {parameterName=ShowSubUV}
  )
  n12 = TextureSampleParameter2D() {
   parameterName=MipRT
   samplerType=SAMPLERTYPE_LinearColor
   texture=&Black_1x1_EXR_Texture
  }
  LinearInterpolate(
   A=LinearInterpolate(A=@n11, B=Multiply(A=@n11, B=@n12.R), Alpha=ScalarParameter() {parameterName=UseClampedOutput})
   B=MaterialFunctionCall!(Vector3=@n12.RGB, MaximumNumberOfDigits, UVs, DebugTextLocation RG_UpperRight BA_LowerLeft, Component Spacing) {materialFunction=/Engine/Functions/Engine_MaterialFunctions02/Utility/DebugFloat3Values}.ColorCodedOutput
   Alpha=ScalarParameter() {parameterName=ShowMipRT}
  )
  ScalarParameter() {defaultValue=2, parameterName=Frame}
 ]
}
