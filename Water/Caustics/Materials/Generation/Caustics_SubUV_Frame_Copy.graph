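# Unlit alpha-composite material for caustics generation, named for copying a frame in a SubUV workflow.
# It samples a linear-color texture parameter RT, defaulting to an engine 1x1 black EXR placeholder.
# The sampled alpha feeds a One Minus node to invert it; the graph shows no connections to material outputs.
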Material /Water/Caustics/Materials/Generation/Caustics_SubUV_Frame_Copy
graph {
 properties {blendMode=BLEND_AlphaComposite, pixelDepthOffsetMode=PDOM_Legacy, shadingModel=MSM_Unlit, usedShadingModels=Unlit}
 nodes [
  n1 = TextureSampleParameter2D() {
   parameterName=RT
   samplerType=SAMPLERTYPE_LinearColor
   texture=/Engine/Functions/Engine_MaterialFunctions02/PivotPainter2/Black_1x1_EXR_Texture
  }
  OneMinus(None=@n1.A)
 ]
}
