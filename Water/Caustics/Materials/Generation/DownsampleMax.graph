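# Unlit material in the water caustics generation set that downsamples the RT texture parameter by taking a maximum.
# It derives texel size from the texture's dimensions, offsets UVs back half a texel, and samples there plus one texel along U and V.
# Only three taps are taken, with no diagonal.
# Their component-wise max collapses to one scalar from R, B and A; R is used twice and G is omitted.
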Material /Water/Caustics/Materials/Generation/DownsampleMax
graph {
 properties {pixelDepthOffsetMode=PDOM_Legacy, shadingModel=MSM_Unlit, usedShadingModels=Unlit}
 nodes [
  n1 = TextureObjectParameter() {
   parameterName=RT
   samplerType=SAMPLERTYPE_LinearColor
   texture=/Engine/Functions/Engine_MaterialFunctions02/PivotPainter2/Black_1x1_EXR_Texture
  }
  n2 = Divide(B=TextureProperty(Tex=@n1)) {desc=Texel Size}
  n3 = Subtract(A=TextureCoordinate(), B=Divide(A=@n2))
  n4 = TextureSampleParameter2D(UVs=@n3) {
   parameterName=RT
   samplerType=SAMPLERTYPE_LinearColor
   texture=&Black_1x1_EXR_Texture
  }
  n5 = Constant2Vector() {r=1}
  n6 = TextureSampleParameter2D(UVs=Add(A=@n3, B=Multiply(A=@n2, B=@n5))) {
   parameterName=RT
   samplerType=SAMPLERTYPE_LinearColor
   texture=&Black_1x1_EXR_Texture
  }
  n7 = Constant2Vector() {g=1}
  n8 = TextureSampleParameter2D(UVs=Add(A=@n3, B=Multiply(A=@n2, B=@n7))) {
   parameterName=RT
   samplerType=SAMPLERTYPE_LinearColor
   texture=&Black_1x1_EXR_Texture
  }
  n9 = MaterialFunctionCall(Float4=Max(A=Max(A=@n4.RGBA, B=@n6.RGBA), B=@n8.RGBA)) {materialFunction=&BreakOutFloat4Components}
  Max(A=Max(A=Max(A=@n9.R, B=@n9.R), B=@n9.B), B=@n9.A)
 ]
}
