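# Unlit alpha-composite material that downsamples the texture bound to the RT parameter, filed under water caustics generation.
# A Custom HLSL node uses the Mode scalar to pick a single sample, a 2x2 per-channel minimum or a 2x2 maximum, returning zero otherwise.
# The 2x2 taps are centred on the UV by a half-texel offset derived from the texture's texel size.
# The result splits into RGB and an inverted alpha.
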Material /Water/Caustics/Materials/Generation/DownsampleRT
graph {
 properties {blendMode=BLEND_AlphaComposite, pixelDepthOffsetMode=PDOM_Legacy, shadingModel=MSM_Unlit, usedShadingModels=Unlit}
 nodes [
  MaterialExpressionTextureCoordinate_1 = TextureCoordinate()
  MaterialExpressionTextureObjectParameter_1 = TextureObjectParameter() {
   parameterName=RT
   samplerType=SAMPLERTYPE_LinearColor
   texture=/Engine/Functions/Engine_MaterialFunctions02/PivotPainter2/Black_1x1_EXR_Texture
  }
  MaterialExpressionTextureProperty_1 = TextureProperty(Tex=@MaterialExpressionTextureObjectParameter_1) {property=TMTM_TexelSize}
  MaterialExpressionScalarParameter_1 = ScalarParameter() {parameterName=Mode}
  n1 = Custom!(
   UV=@MaterialExpressionTextureCoordinate_1
   RT=@MaterialExpressionTextureObjectParameter_1
   TexelSize=@MaterialExpressionTextureProperty_1
   Mode=@MaterialExpressionScalarParameter_1
  ) {
   code="\r\nif (Mode == 0.0)\r\n{\r\n\treturn RT.SampleLevel(RTSampler, UV, 0);\r\n}\r\n\r\nif (Mode == 1.0)\r\n{\r\n\tUV -= TexelSize / 2.0;\r\n\t\r\n\tfloat4 t0 = RT.SampleLevel(RTSampler, UV, 0);\r\n\tfloat4 t1 = RT.SampleLevel(RTSampler, UV + TexelSize * float2(1,0), 0);\r\n\tfloat4 t2 = RT.SampleLevel(RTSampler, UV + TexelSize * float2(0,1), 0);\r\n\tfloat4 t3 = RT.SampleLevel(RTSampler, UV + TexelSize * float2(1,1), 0);\r\n\t\r\n\tfloat4 outmax = min(t0, min(t1, min(t2, t3)));\r\n\treturn outmax;\r\n\t\r\n}\r\n\r\nif (Mode == 2.0)\r\n{\r\n\tUV -= TexelSize / 2.0;\r\n\t\r\n\tfloat4 t0 = RT.SampleLevel(RTSampler, UV, 0);\r\n\tfloat4 t1 = RT.SampleLevel(RTSampler, UV + TexelSize * float2(1,0), 0);\r\n\tfloat4 t2 = RT.SampleLevel(RTSampler, UV + TexelSize * float2(0,1), 0);\r\n\tfloat4 t3 = RT.SampleLevel(RTSampler, UV + TexelSize * float2(1,1), 0);\r\n\t\r\n\tfloat4 outmax = max(t0, max(t1, max(t2, t3)));\r\n\treturn outmax;\r\n}\r\n\r\n//for undefined Mode\r\nreturn 0.0;"
   description=Downsample RT
   outputType=CMOT_Float4
   inputs [
    {inputName=UV, input {expression=&self:MaterialExpressionTextureCoordinate_1}}
    {inputName=RT, input {expression=&self:MaterialExpressionTextureObjectParameter_1}}
    {inputName=TexelSize, input {expression=&self:MaterialExpressionTextureProperty_1}}
    {inputName=Mode, input {expression=&self:MaterialExpressionScalarParameter_1}}
   ]
  }
  ComponentMask(None=@n1) {b=true, g=true, r=true}
  OneMinus(None=ComponentMask(None=@n1) {a=true})
 ]
}
