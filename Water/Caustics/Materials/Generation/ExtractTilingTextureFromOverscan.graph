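# Unlit caustics-generation material that extracts a tiling texture from a render target padded with overscan.
# UVs shrink around the center by the resolution-to-padded-size ratio, blended by Scale Back, to crop the border.
# A custom HLSL node samples RT offset one tile in X, Y and diagonally, masking the sum to edges.
# That wrapped overscan, scaled by Show Tiled Edges, is added to the cropped sample, making opposite edges tile.
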Material /Water/Caustics/Materials/Generation/ExtractTilingTextureFromOverscan
graph {
 properties {pixelDepthOffsetMode=PDOM_Legacy, shadingModel=MSM_Unlit, usedShadingModels=Unlit}
 nodes [
  n1 = Subtract(A=TextureCoordinate()) {constB=0.5}
  n2 = ScalarParameter() {defaultValue=16, parameterName=Overscan Quads}
  n3 = ScalarParameter() {defaultValue=1024, parameterName=Resolution}
  n4 = Divide(A=Add(A=Multiply(A=@n2) {constB=2}, B=@n3), B=@n3)
  MaterialExpressionLinearInterpolate_1 = LinearInterpolate(A=TextureCoordinate(), B=Add(A=Divide(A=@n1, B=@n4)) {constB=0.5}, Alpha=ScalarParameter() {defaultValue=1, parameterName=Scale Back})
  MaterialExpressionTextureObjectParameter_1 = TextureObjectParameter() {
   parameterName=RT
   samplerType=SAMPLERTYPE_LinearColor
   texture=/Engine/Functions/Engine_MaterialFunctions02/PivotPainter2/Black_1x1_EXR_Texture
  }
  MaterialExpressionScalarParameter_2 = ScalarParameter() {defaultValue=1024, parameterName=Resolution}
  MaterialExpressionScalarParameter_5 = ScalarParameter() {defaultValue=16, parameterName=Overscan Quads}
  n5 = Custom!(
   UV=@MaterialExpressionLinearInterpolate_1
   Tex=@MaterialExpressionTextureObjectParameter_1
   Resolution=@MaterialExpressionScalarParameter_2
   Overscan=@MaterialExpressionScalarParameter_5
  ) {
   code="\r\n\r\nconst float scalefactor = Resolution / (Resolution + (Overscan * 2.0));\r\n\r\nfloat2 offset = lerp(1.0, -1.0, saturate(ceil(UV - 0.5)));\r\n\r\nfloat2 UVx = saturate(UV + float2(scalefactor * offset.x, 0.0));\r\nfloat2 UVy = saturate(UV + float2(0.0, scalefactor * offset.y));\r\n\r\nfloat2 UVxy =  saturate(UV + float2(scalefactor * offset.x, scalefactor * offset.y));\r\n\r\nfloat3 TileX = Tex.SampleLevel(TexSampler, UVx, 0);\r\nfloat3 TileY = Tex.SampleLevel(TexSampler, UVy, 0);\r\nfloat3 TileXY = Tex.SampleLevel(TexSampler, UVxy, 0);\r\n\r\nfloat2 boxmask = abs(UV - 0.5);\r\n\r\nfloat bound = (Resolution - (Overscan * 2)) / Resolution;\r\nfloat falloff = Overscan / Resolution;\r\nbound -= 2.0 * falloff;\r\n\r\nboxmask -= bound * 0.5;\r\n\r\nboxmask = max(0, boxmask);\r\nfloat outmask = length(boxmask) / falloff;\r\noutmask = saturate(outmask);\r\n\r\n//return outmask;\r\nreturn (TileX + TileY + TileXY) * outmask;"
   description=Tiling
   inputs [
    {inputName=UV, input {expression=&self:MaterialExpressionLinearInterpolate_1}}
    {
     inputName=Tex
     input {expression=&self:MaterialExpressionTextureObjectParameter_1}
    }
    {
     inputName=Resolution
     input {expression=&self:MaterialExpressionScalarParameter_2}
    }
    {
     inputName=Overscan
     input {expression=&self:MaterialExpressionScalarParameter_5}
    }
   ]
  }
  n6 = TextureSampleParameter2D(UVs=@MaterialExpressionLinearInterpolate_1) {
   parameterName=RT
   samplerType=SAMPLERTYPE_LinearColor
   texture=&Black_1x1_EXR_Texture
  }
  Add(A=Multiply(A=@n5, B=ScalarParameter() {defaultValue=1, parameterName=Show Tiled Edges}), B=@n6.RGB)
 ]
}
