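# Unlit material that computes a spectrum table, apparently for water caustics, spreading RGB bands along U by a focus amount taken from inverted V.
# A Custom HLSL loop accumulates spectrum samples under linear falloff masks, with focus setting their spacing, start point and radius.
# A second loop sums the whole spectrum to normalise the result, which feeds a tiled DebugFloat3Values readout; no material output connection is shown.
# A third loop labelled "Bugged" is left unused.
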Material /Water/Caustics/Materials/Generation/GenerateSpectrumTable
graph {
 properties {pixelDepthOffsetMode=PDOM_Legacy, shadingModel=MSM_Unlit, usedShadingModels=Unlit}
 nodes [
  MaterialExpressionTextureCoordinate_2 = TextureCoordinate()
  MaterialExpressionScalarParameter_2 = ScalarParameter() {defaultValue=256, parameterName=Steps}
  Custom!(UV=@MaterialExpressionTextureCoordinate_2, MaxSteps=@MaterialExpressionScalarParameter_2) {
   code="float3 color = 0.0;\r\nfloat uv = 0.0;\r\nfloat ss = 1.0/MaxSteps;\r\n\r\nfor(int i = 0; i < MaxSteps;  i++)\r\n{\r\n\r\n\tuv += ss;\r\n\tcolor += saturate(1-abs(2*uv - float3(0,1,2)));\r\n\t\r\n\r\n\t\r\n}\r\n\r\nreturn color;"
   description=Bugged
   inputs [
    {inputName=UV, input {expression=&self:MaterialExpressionTextureCoordinate_2}}
    {inputName=MaxSteps, input {expression=&self:MaterialExpressionScalarParameter_2}}
   ]
  }
  n1 = TextureCoordinate()
  MaterialExpressionAppendVector_12 = AppendVector(A=OneMinus(None=ComponentMask(None=@n1) {r=true}), B=Constant() {r=0.5})
  MaterialExpressionScalarParameter_78 = ScalarParameter() {defaultValue=256, parameterName=Steps}
  n2 = OneMinus(None=ComponentMask(None=TextureCoordinate()) {g=true})
  n3 = Subtract(A=Constant() {r=0.5}, B=Divide(A=@n2))
  MaterialExpressionMultiply_33 = Multiply(A=@MaterialExpressionScalarParameter_78, B=@n3)
  MaterialExpressionAppendVector_11 = AppendVector(A=@n3, B=Constant() {r=0.5})
  MaterialExpressionReroute_0 = Reroute(None=@n2)
  n4 = Custom!(
   UV=@MaterialExpressionAppendVector_12
   MaxSteps=@MaterialExpressionScalarParameter_78
   Scale=@MaterialExpressionMultiply_33
   start=@MaterialExpressionAppendVector_11
   focus=@MaterialExpressionReroute_0
  ) {
   code="\r\nfloat2 curUV = start;\r\nfloat stepsize = 1 / MaxSteps;\r\nfloat3 accum = 0;\r\n\r\nint i = 0;\r\n\r\nwhile (i < MaxSteps)\r\n{\r\n\r\n\t\r\nfloat xlocation = (float)i * stepsize;\r\n\r\n\tfloat3 spectrumsample = saturate(1-abs(2*xlocation- float3(0,1,2)));\r\n\r\n\tfloat spheremask = length(UV - curUV);\r\n\tspheremask *= MaxSteps / Scale;\r\n\tspheremask = pow(1 - saturate(spheremask) , 1);\r\n\r\n\taccum += spheremask * spectrumsample;\r\n\r\n\tcurUV += float2(stepsize * focus, 0);\r\n\t\r\n\ti++;\r\n}\r\n\r\nreturn accum;\r\n"
   inputs [
    {inputName=UV, input {expression=&self:MaterialExpressionAppendVector_12}}
    {inputName=MaxSteps, input {expression=&self:MaterialExpressionScalarParameter_78}}
    {inputName=Scale, input {expression=&self:MaterialExpressionMultiply_33}}
    {inputName=start, input {expression=&self:MaterialExpressionAppendVector_11}}
    {inputName=focus, input {expression=&self:MaterialExpressionReroute_0}}
   ]
  }
  MaterialExpressionScalarParameter_1 = ScalarParameter() {defaultValue=256, parameterName=Steps}
  n5 = Custom!(MaxSteps=@MaterialExpressionScalarParameter_1) {
   code="float3 color = 0.0;\r\nfloat ss = 1.0/MaxSteps;\r\n\r\nfor(int i = 0; i < MaxSteps;  i++)\r\n{\r\n\r\n    float xpos = (float)i * ss;\r\n\tcolor += saturate(1-abs(2*xpos - float3(0,1,2)));\r\n\r\n\t\r\n}\r\n\r\nreturn color;"
   description=Works
   inputs [{inputName=MaxSteps, input {expression=&self:MaterialExpressionScalarParameter_1}}]
  }
  n6 = TextureCoordinate()
  n7 = ScalarParameter() {defaultValue=16, parameterName=Text Repeat}
  MaterialFunctionCall!(
   Vector3=Divide(A=@n4, B=@n5) {constB=128}
   MaximumNumberOfDigits
   UVs=Frac(None=Multiply(A=@n6, B=@n7))
   DebugTextLocation RG_UpperRight BA_LowerLeft
   Component Spacing
  ) {materialFunction=/Engine/Functions/Engine_MaterialFunctions02/Utility/DebugFloat3Values}
 ]
}
