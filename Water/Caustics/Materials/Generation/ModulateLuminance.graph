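# Unlit Modulate-blend material in the water caustics generation set, built to compute a reciprocal luminance scale, apparently for normalizing caustics brightness.
# It samples the MipRT texture and takes the larger of red and blue as luminance; red is compared with itself, so green is ignored.
# A parameter can blend in a manual value instead.
# The result is inverted, with a small floor preventing division by zero.
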Material /Water/Caustics/Materials/Generation/ModulateLuminance
graph {
 properties {blendMode=BLEND_Modulate, pixelDepthOffsetMode=PDOM_Legacy, shadingModel=MSM_Unlit, usedShadingModels=Unlit}
 nodes [
  n1 = TextureSampleParameter2D() {
   parameterName=MipRT
   samplerType=SAMPLERTYPE_LinearColor
   texture=/Engine/Functions/Engine_MaterialFunctions02/PivotPainter2/Black_1x1_EXR_Texture
  }
  n2 = Max(A=Max(A=@n1.R, B=@n1.R), B=@n1.B)
  n3 = VectorParameter() {parameterName=Manual Value}
  n4 = ScalarParameter() {parameterName=Use Manual Value}
  Divide(B=Max(A=LinearInterpolate(A=@n2, B=@n3.RGB, Alpha=@n4)) {constB=0.0001})
 ]
}
