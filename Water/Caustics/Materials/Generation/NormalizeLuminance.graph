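# Unlit alpha-composite material in the water caustics generation set that normalizes the luminance of a render target.
# It divides the RGB of the RT texture parameter by that of MipRT, which by name is a mipped, averaged copy, and inverts RT's alpha.
# Both parameters default to a black 1x1 EXR placeholder, and a stray AnalogHat texture sample feeds nothing.
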Material /Water/Caustics/Materials/Generation/NormalizeLuminance
graph {
 properties {blendMode=BLEND_AlphaComposite, pixelDepthOffsetMode=PDOM_Legacy, shadingModel=MSM_Unlit, usedShadingModels=Unlit}
 nodes [
  n1 = TextureSampleParameter2D() {
   parameterName=RT
   samplerType=SAMPLERTYPE_LinearColor
   texture=/Engine/Functions/Engine_MaterialFunctions02/PivotPainter2/Black_1x1_EXR_Texture
  }
  n2 = TextureSampleParameter2D() {
   parameterName=MipRT
   samplerType=SAMPLERTYPE_LinearColor
   texture=&Black_1x1_EXR_Texture
  }
  Divide(A=@n1.RGB, B=@n2.RGB)
  OneMinus(None=@n1.A)
  TextureSample() {texture=/Engine/MobileResources/HUD/AnalogHat}
 ]
}
