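# Unlit alpha-composite material that generates a tiling Gerstner wave normal for water caustics generation.
# It samples GetGerstnerWaves at centered and half-tile wrapped positions, blending the shifted versions along the center cross with sphere masks.
# A custom soft-edge mask fades the centered sample into that blend at the borders, hiding seams.
# Time follows MPC_Water's freezable clock, and an S-curved FrameAlpha scales the result for frame blending.
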Material /Water/Caustics/Materials/Generation/RenderTilingWaterNormal
graph {
 properties {
  bAllowNegativeEmissiveColor=true
  blendMode=BLEND_AlphaComposite
  pixelDepthOffsetMode=PDOM_Legacy
  shadingModel=MSM_Unlit
  usedShadingModels=Unlit
 }
 nodes [
  n1 = TextureCoordinate()
  n2 = Subtract(A=Frac(None=Add(A=@n1) {constB=0.5})) {constB=0.5}
  n3 = ScalarParameter() {defaultValue=2048, parameterName=WorldSize}
  n4 = Fmod(
   A=LinearInterpolate(
    A=Time()
    B=CollectionParameter() {collection=/Water/Materials/MPC/MPC_Water, parameterName=ManualTime}
    Alpha=CollectionParameter() {collection=&MPC_Water, parameterName=FreezeTime}
   )
   B=Constant() {r=1000}
  )
  n5 = Constant2Vector() {g=0.5}
  n6 = Subtract(A=Frac(None=Add(A=@n1, B=@n5))) {constB=0.5}
  n7 = ScalarParameter() {defaultValue=2048, parameterName=WorldSize}
  n8 = MaterialFunctionCall(Float2=TextureCoordinate()) {materialFunction=&BreakOutFloat2Components}
  n9 = Constant() {r=0.5}
  n10 = ScalarParameter() {defaultValue=0.1, parameterName=CrossFixWidth}
  n11 = LinearInterpolate(
   A=MaterialFunctionCall!(WaveParameterRT, Override Position=Multiply(A=@n2, B=@n3), Time=@n4) {materialFunction=/Water/Caustics/Materials/Functions/GetGerstnerWaves}.Normal
   B=MaterialFunctionCall(Override Position=Multiply(A=@n6, B=@n7), Time=@n4) {materialFunction=&GetGerstnerWaves}.Normal
   Alpha=SphereMask(A=@n8.R, B=@n9, Radius=@n10) {hardnessPercent=0}
  )
  n12 = Frac(None=Add(A=@n1, B=Constant2Vector() {r=0.5}))
  n13 = MaterialFunctionCall(Override Position=Multiply(A=Subtract(A=@n12) {constB=0.5}, B=ScalarParameter() {defaultValue=2048, parameterName=WorldSize}), Time=@n4) {materialFunction=&GetGerstnerWaves}
  n14 = SphereMask(A=@n8.G, B=@n9, Radius=@n10) {hardnessPercent=0}
  n15 = Multiply(A=Subtract(A=@n1) {constB=0.5}, B=ScalarParameter() {defaultValue=2048, parameterName=WorldSize})
  MaterialExpressionScalarParameter_23 = ScalarParameter() {defaultValue=0.25, parameterName=sm}
  MaterialExpressionTextureCoordinate_10 = TextureCoordinate()
  n16 = Custom!(SplineMargin=@MaterialExpressionScalarParameter_23, UV=@MaterialExpressionTextureCoordinate_10) {
   code="//input float HeightScale;\r\n//input float SplineMargin;\r\n\r\nfloat SplineParameter1 = (0.5f - SplineMargin);\r\nfloat SplineParameter2 = (1.0f/SplineMargin);\r\n\r\n\r\n    // Re-scale [0..0.1] and [0.9..1] to [0..1]\r\n    float2 T        = max(abs(UV - 0.5f) - SplineParameter1, 0.0f) * SplineParameter2;\r\n    // Cubic interpolation\r\n    float2 F        = 2*T*T*T - 3*T*T + 1;\r\n    // Bilinear filter from 2D to 1D factor (making it zero at the edges)\r\n    float Scale        =  F.x * F.y;\r\n\r\nreturn Scale;"
   description=Soft Edge Mask
   inputs [
    {inputName=SplineMargin, input {expression=&self:MaterialExpressionScalarParameter_23}}
    {inputName=UV, input {expression=&self:MaterialExpressionTextureCoordinate_10}}
   ]
  }
  Multiply(
   A=LinearInterpolate(
    A=LinearInterpolate(A=@n11, B=@n13.Normal, Alpha=@n14)
    B=MaterialFunctionCall(Override Position=@n15, Time=@n4) {materialFunction=&GetGerstnerWaves}.Normal
    Alpha=@n16
   )
   B=MaterialFunctionCall(In=ScalarParameter() {parameterName=FrameAlpha}, Power=ScalarParameter() {defaultValue=3, parameterName=Frame Blend Contrast}) {materialFunction=&SCurve}.Result
  )
  TextureObject() {samplerType=SAMPLERTYPE_LinearColor, texture=/Water/Textures/DefaultWaveSpectrumRT}
 ]
}
